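# Niagara spawn burst for the Private Jet emote, apparently the front-end version, combining smoke, a brief air glow and "Spooders" mesh particles.
# Smoke uses animated SubUV sprites with dynamic material parameters, while Spooders_Low meshes fly out under gravity.
# GPU Spooders_High_FE spawns only in front end, colliding with an analytical plane then switching to noise- and attraction-driven crawling.
# Both sparkle emitters and gameplay Spooders_High, with distance-field surface crawling, are disabled.
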NiagaraSystem /BRCosmetics/Effects/Fort_Effects/Emotes/PrivateJet/NS_Emote_Private_Jet_SpawnFE
dependencies {
 usedDataInterfaces [
  &NiagaraDataInterfaceCurve
  &NiagaraDataInterfaceSpriteRendererInfo
  &NiagaraDataInterfaceArrayColor
  &NiagaraDataInterfaceCollisionQuery
 ]
 usedDynamicInputs []
 usedModules [&SystemState]
}
emitters {
 AirGlow {
  emitterUpdateScript [
   &EmitterState()
   &SpawnBurst_Instantaneous(Spawn Count=1, Use Spawn Probability=false)
  ]
  particleSpawnScript [
   &InitializeParticle(
    Color=<rgba 0.4323 0.4323 0.4323 0.2>
    Lifetime=0.25
    Sprite Size Mode=::NewEnumerator3
    Uniform Sprite Size=300
    UsePositionOffset=false
    Write Material Random=true
   ) {module=InitializeParticle001}
   &DynamicMaterialParameters(
    Index 0 Param 1=2
    Index 0 Param 2=10
    Index 0 Param 3=1
    Index 0 Param 4=1
    Module.Param0WriteEnabled=true
    Module.Param1WriteEnabled=true
    Module.Param2WriteEnabled=true
    Module.Param3WriteEnabled=true
   )
   &CameraOffset(Camera Offset Amount=80)
   &SystemLocation(MaskSpawnGroup=false, Offset=&FN_FrontEndVector3Modifier(Gameplay=<0 0 0>, Front End=<0 0 0>))
  ]
  particleUpdateScript [
   &ParticleState()
   &ScaleSpriteSize(Scale Factor=&Vector2DFromFloat(Value=&FloatFromCurve(FloatCurve=DataInterface() {propertyValues {Curve {keys [<at 0 1 RCIM_Cubic>]}}})))
   &ScaleColor(
    Scale Alpha=&FloatFromCurve(FloatCurve=DataInterface() {propertyValues {Curve {keys [<at 0 0>,<at 0.25 0.75>,<at 1 0>]}}})
    ScaleA=true
    ScaleRGB=true
   ) {module=ScaleColor001}
  ]
  rendererClasses [&NiagaraSpriteRendererProperties]
  renderers [{rendererClass=&NiagaraSpriteRendererProperties}]
 }
 LargeSparkles001 {
  bEnabled=false
  emitterUpdateScript [
   &EmitterState()
   &SpawnBurst_Instantaneous(Spawn Count=25, Use Spawn Probability=false, UseLoopCountLimit=false)
  ]
  particleSpawnScript [
   &InitializeParticle(
    Color=&MultiplyLinearColor(
     Linear Color=&SelectLinearColorFromArray(
      Array Sampling Mode=::NewEnumerator0
      Direct Array Index=?
      Color Selection Array=DataInterface() {propertyValues {ColorData [<rgba 1 0 0 1>, <rgba 0 1 0 1>, <rgba 0.4662 0.8486 1 1>, <rgba 1 1 0 1>]}}
      Random Seed=0
     )
     Scale Factor=&MakeFloat4FromVectorAndFloat("Vector (RGB)"=&VectorFromFloat(Value=25), "Float (Alpha)"=0.25)
    )
    Lifetime=&UniformRangedFloat(Minimum=0.25, Maximum=0.6)
    Sprite Rotation Angle=45
    Sprite Rotation Mode=::NewEnumerator0
    Sprite Size Mode=::NewEnumerator3
    Uniform Sprite Size=&UniformRangedFloat(Minimum=220, Maximum=250)
    UsePositionOffset=false
    Write Material Random=true
   ) {module=InitializeParticle001}
   &SystemLocation(MaskSpawnGroup=false, Offset=&FN_FrontEndVector3Modifier(Gameplay=<0 0 100>, Front End=<0 0 70>)) {enabled=false}
   &ShapeLocation(
    MaskSpawnGroup=false
    Offset Mode=::NewEnumerator0
    Override Seed=false
    Sphere Radius=15
    Sphere Surface Distribution=0.2
    UseOwnerScale=true
   )
   &ShapeLocation(
    Cylinder Height=500
    MaskSpawnGroup=false
    Offset=<50 0 0>
    Offset Coordinate Space=::Local
    Offset Mode=::NewEnumerator0
    Override Seed=false
    Rotation Mode=::NewEnumerator1
    Shape Primitive=::NewEnumerator1
    Surface Only=false
    UseOwnerScale=true
   ) {module=ShapeLocation001}
   &AddVelocity(
    Velocity=<1 0 0>
    Velocity Speed Scale=&Lerp_Float(
     A=500
     B=1200
     Alpha=&FloatFromCurve(
      FloatCurve=DataInterface() {propertyValues {Curve {keys [
          {interpMode=RCIM_Cubic, tangentMode=RCTM_User, time=0, value=0}
          {arriveTangent=1.7817, interpMode=RCIM_Cubic, leaveTangent=1.7817, tangentMode=RCTM_User, time=1, value=1}
         ]}}}
      CurveIndex=&RandomRangeFloat(Minimum=0, Maximum=1)
     )
    )
   )
   &AddVelocity(
    Module.Use Falloff Distance=false
    Override Seed=false
    Velocity Mode=::NewEnumerator1
    Velocity Origin=$Particles.ShapeLocation.ShapePosition
    Velocity Speed=&Lerp_Float(
     A=700
     B=1500
     Alpha=&FloatFromCurve(
      FloatCurve=DataInterface() {propertyValues {Curve {keys [
          {interpMode=RCIM_Cubic, tangentMode=RCTM_User, time=0, value=0}
          {arriveTangent=1.7817, interpMode=RCIM_Cubic, leaveTangent=1.7817, tangentMode=RCTM_User, time=1, value=1}
         ]}}}
      CurveIndex=&RandomRangeFloat(Minimum=0, Maximum=1)
     )
    )
   ) {module=AddVelocity001}
  ]
  particleUpdateScript [
   &ParticleState()
   &GravityForce(Gravity=<0 0 200>)
   /Niagara/Modules/Update/Forces/CurlNoiseForce(Noise Frequency=0.02, Noise Quality / Cost=::NewEnumerator0, Noise Strength=500, Pan Noise=false)
   &Drag(Drag=8, Use Linear Drag=true, Use Rotational Drag=true)
   &SolveForcesAndVelocity(Clamp Velocity=false, Limit Acceleration=false)
   &ScaleColor(
    Scale Alpha=&FloatFromCurve(FloatCurve=DataInterface() {propertyValues {Curve {keys [
         {arriveTangent=-2.9626, interpMode=RCIM_Cubic, leaveTangent=-2.9626, tangentMode=RCTM_User, time=0.3, value=1}
         {tangentMode=RCTM_User, time=1, value=0}
        ]}}})
    ScaleA=true
    ScaleRGB=true
   )
   &ScaleSpriteSize()
  ]
  rendererClasses [&NiagaraSpriteRendererProperties]
  renderers [{rendererClass=&NiagaraSpriteRendererProperties}]
 }
 Smoke {
  emitterUpdateScript [
   &EmitterState()
   &SpawnBurst_Instantaneous(Spawn Count=25, Use Spawn Probability=false)
  ]
  particleSpawnScript [
   &InitializeParticle(
    Color Maximum=<rgba 1 1 1 2>
    Color Minimum=<rgba 1 1 1 3>
    Color Mode=::NewEnumerator2
    Lifetime Max=0.55
    Lifetime Min=0.2
    Lifetime Mode=::NewEnumerator1
    Sprite Rotation Mode=::NewEnumerator2
    Sprite Size Mode=::NewEnumerator3
    Uniform Sprite Size=&Multiply_Float(A=2, B=&RandomRangeFloat(Minimum=30, Maximum=50))
    UsePositionOffset=false
    Write Material Random=true
   ) {module=InitializeParticle001}
   &SystemLocation(MaskSpawnGroup=false, Offset=&FN_FrontEndVector3Modifier(Gameplay=<0 0 100>, Front End=<0 0 70>)) {enabled=false}
   &ShapeLocation(MaskSpawnGroup=false, Offset=<0 0 25>, Offset Mode=::NewEnumerator0, Override Seed=false, Sphere Radius=1, UseOwnerScale=true)
   &ShapeLocation(
    Cylinder Radius=25
    MaskSpawnGroup=false
    Offset Coordinate Space=::Local
    Offset Mode=::NewEnumerator0
    Override Seed=false
    Rotation Axis=<0 1 0>
    Rotation Mode=::NewEnumerator1
    Shape Primitive=::NewEnumerator1
    Surface Only=false
    UseOwnerScale=true
   ) {module=ShapeLocation001}
   &CameraOffset(Camera Offset Amount=&RandomRangeFloat(Minimum=25, Maximum=50)) {module=CameraOffset001}
   &AddVelocity(Velocity=<0 0 100>) {module=AddVelocity002}
   &AddVelocity(Module.Use Falloff Distance=false, Origin Offset=<0 0 25>, Override Seed=false, Velocity Mode=::NewEnumerator1, Velocity Speed=300)
   &AddVelocity(
    Velocity=<1 0 0>
    Velocity Speed Scale=&Lerp_Float(
     A=500
     B=2500
     Alpha=&FloatFromCurve(
      FloatCurve=DataInterface() {propertyValues {Curve {keys [
          {interpMode=RCIM_Cubic, tangentMode=RCTM_User, time=0, value=0}
          {arriveTangent=1.7817, interpMode=RCIM_Cubic, leaveTangent=1.7817, tangentMode=RCTM_User, time=1, value=1}
         ]}}}
      CurveIndex=&RandomRangeFloat(Minimum=0, Maximum=1)
     )
    )
   ) {enabled=false, module=AddVelocity001}
   /SetVariables(Particles.PivotOffset:Vector2f=&RandomRangeVector2D(Minimum=<0.4 0.4>, Maximum=<0.5 0.5>)) {isSetParameters=true}
  ]
  particleUpdateScript [
   &ParticleState()
   &Drag(Drag=&RandomRangeFloat(Minimum=9, Maximum=13), Use Linear Drag=true, Use Rotational Drag=true)
   &SpriteRotationRate(Rotation Rate=&Multiply_Float(
     A=&RandomRangeFloat(Minimum=-2, Maximum=2)
     B=&FloatFromCurve(
      FloatCurve=DataInterface() {propertyValues {Curve {keys [
          {arriveTangent=-4.0107, interpMode=RCIM_Cubic, leaveTangent=-4.0107, tangentMode=RCTM_User, time=-0.0046, value=1.0113}
          {arriveTangent=-0.2496, interpMode=RCIM_Cubic, leaveTangent=-0.2496, tangentMode=RCTM_User, time=0.5, value=0.16}
         ]}}}
      CurveIndex=$Emitter.Age
      Scale Curve=150
     )
    ))
   &ScaleSpriteSize(Uniform Curve Sprite Scale=DataInterface() {propertyValues {Curve {keys [
        {arriveTangent=2.6959, interpMode=RCIM_Cubic, leaveTangent=2.6959, tangentMode=RCTM_User, time=0, value=0}
        {arriveTangent=0.5869, interpMode=RCIM_Cubic, leaveTangent=0.5869, tangentMode=RCTM_User, time=0.3, value=0.8}
        {arriveTangent=0.0307, interpMode=RCIM_Cubic, leaveTangent=0.0307, tangentMode=RCTM_User, time=1, value=1}
       ]}}})
   &InheritSourceMovement(ApplyForce=false, ApplyPosition=true, ApplyPositionFalloff=false) {enabled=false}
   &SolveForcesAndVelocity(Clamp Velocity=false, Limit Acceleration=false)
   &ScaleColor(
    Scale Alpha=&Multiply_Float(
     A=&FloatFromCurve(
      FloatCurve=DataInterface() {propertyValues {Curve {keys [
          {arriveTangent=6.0302, interpMode=RCIM_Cubic, leaveTangent=6.0302, tangentMode=RCTM_User, time=0, value=0.4}
          {arriveTangent=-0.2604, interpMode=RCIM_Cubic, leaveTangent=-0.2604, tangentMode=RCTM_User, time=0.1, value=1}
          {arriveTangent=-0.484, interpMode=RCIM_Cubic, leaveTangent=-0.484, tangentMode=RCTM_User, time=1, value=0}
         ]}}}
      Scale Curve=&RandomRangeFloat(Minimum=0.5, Maximum=1)
     )
     B=&FN_FrontEndFloatModifier(Gameplay=1, Front End=0.5)
    )
    Scale RGB=&VectorFromFloat(Value=1)
    ScaleA=true
    ScaleRGB=true
   )
   &DynamicMaterialParameters(
    Index 0 Param 1=&Add_Float(
     A=$Particles.RandomFloat
     B=&FloatFromCurve(
      FloatCurve=DataInterface() {propertyValues {Curve {keys [
          {arriveTangent=2.7533, interpMode=RCIM_Cubic, leaveTangent=1.8958, tangentMode=RCTM_User, time=0, value=1}
          {arriveTangent=0.8292, interpMode=RCIM_Cubic, leaveTangent=0.8292, tangentMode=RCTM_User, time=0.321, value=1.3453}
          {arriveTangent=0.0529, interpMode=RCIM_Cubic, leaveTangent=0.2021, tangentMode=RCTM_User, time=1, value=1.5}
         ]}}}
      CurveIndex=&Add_Float(A=&Multiply_Float(A=$Particles.Initial.MaterialRandom, B=0.5), B=$Particles.NormalizedAge)
     )
    )
    Index 0 Param 2=&FloatFromCurve(FloatCurve=DataInterface() {propertyValues {Curve {keys [
         {arriveTangent=0.0453, interpMode=RCIM_Cubic, tangentMode=RCTM_Break, time=0, value=-1}
         {arriveTangent=0.0604, interpMode=RCIM_Cubic, leaveTangent=0.0604, tangentMode=RCTM_User, time=1, value=-1}
        ]}}})
    Index 0 Param 3=&FloatFromCurve(FloatCurve=DataInterface() {propertyValues {Curve {keys [
         {arriveTangent=-53.9199, interpMode=RCIM_Cubic, leaveTangent=-8.9727, tangentMode=RCTM_User, time=0, value=1}
         {arriveTangent=0.2347, interpMode=RCIM_Cubic, leaveTangent=0.2723, tangentMode=RCTM_Break, time=0.25, value=0}
        ]}}})
    Index 0 Param 4=&FloatFromCurve(FloatCurve=DataInterface() {propertyValues {Curve {keys [<at 0.5 1>,<at 1 0>]}}})
    Index 2 Param 1=&FloatFromCurve(FloatCurve=DataInterface() {propertyValues {Curve {keys [
         {interpMode=RCIM_Cubic, leaveTangent=4.2485, tangentMode=RCTM_Break, time=0, value=0}
         {arriveTangent=0.8868, interpMode=RCIM_Cubic, leaveTangent=0.8868, tangentMode=RCTM_User, time=0.17, value=0.6}
         <at 1 1>
        ]}}})
    Index 2 Param 2=&FloatFromCurve(
     FloatCurve=DataInterface() {propertyValues {Curve {keys [
         {arriveTangent=0.9316, interpMode=RCIM_Cubic, leaveTangent=0.9316, tangentMode=RCTM_User, time=0, value=0.1}
         {arriveTangent=0.0259, interpMode=RCIM_Cubic, leaveTangent=0.0259, tangentMode=RCTM_User, time=1, value=0.4}
        ]}}}
     Scale Curve=&RandomRangeFloat(Minimum=0.2, Maximum=0.7)
    )
    Index 2 Param 3=&FloatFromCurve(FloatCurve=DataInterface() {propertyValues {Curve {keys [
         {interpMode=RCIM_Cubic, leaveTangent=2.3064, tangentMode=RCTM_Break, time=0, value=1}
         {interpMode=RCIM_Cubic, leaveTangent=11.1233, tangentMode=RCTM_Break, time=1, value=2}
        ]}}})
    Index 2 Param 4=&FloatFromCurve(
     FloatCurve=DataInterface() {propertyValues {Curve {keys [
         {arriveTangent=-0.6867, interpMode=RCIM_Cubic, leaveTangent=-0.6867, tangentMode=RCTM_User, time=0, value=1}
         {interpMode=RCIM_Cubic, leaveTangent=-1.2956, tangentMode=RCTM_Break, time=1, value=0.75}
        ]}}}
     Scale Curve=&RandomRangeFloat(Minimum=1, Maximum=1.6)
    )
    Module.Index0Param0Write=true
    Module.Index0Param1Write=true
    Module.Index0Param2Write=true
    Module.Index0Param3Write=true
    Module.Index2Param0Write=true
    Module.Index2Param1Write=true
    Module.Index2Param2Write=true
    Module.Index2Param3Write=true
    Write Parameter Index 2=true
   )
   &SubUVAnimation(
    Random Start Frame=true
    "Sprite Renderer "=DataInterface() {propertyValues {SpriteRenderer=&self:Smoke_0.NiagaraSpriteRendererProperties_2}}
    SubUV Lookup Index Scale=0.3
    UseEndFrame=false
    UseStartFrame=false
   )
   /SetVariables(Particles.SubImageIndex2:NiagaraFloat=$Particles.SubImageIndex) {isSetParameters=true}
   /SetVariables(
    Particles.Vector 2D:Vector2f=&Multiply_Vector2D_ByFloat(Float=1.6294, Vector2D=$Particles.SpriteSize)
    Particles.Vector 4:Vector4f=&MakeVector4(
     X=&FloatFromCurve(FloatCurve=DataInterface() {propertyValues {Curve {keys [<at 0 1>,<at 1 0>]}}})
     Y=1
     Z=0
     W=1
    )
   ) {isSetParameters=true}
  ]
  rendererClasses [&NiagaraSpriteRendererProperties]
  renderers [{rendererClass=&NiagaraSpriteRendererProperties}]
 }
 Sparkles {
  bEnabled=false
  emitterUpdateScript [
   &EmitterState()
   &SpawnBurst_Instantaneous(Spawn Count=75, Use Spawn Probability=false, UseLoopCountLimit=false)
  ]
  particleSpawnScript [
   &InitializeParticle(
    Color=&MultiplyLinearColor(
     Linear Color=&SelectLinearColorFromArray(
      Array Sampling Mode=::NewEnumerator0
      Direct Array Index=?
      Color Selection Array=DataInterface() {propertyValues {ColorData [<rgba 1 0 0 1>, <rgba 1 0 0 1>, <rgba 1 0 0 1>, <rgba 0 1 0 1>, <rgba 0 1 0 1>, <rgba 0 1 0 1>, <rgba 1 0.3 0 1>, <rgba 1 0.3 0 1>]}}
      Random Seed=0
     )
     Scale Factor=&MakeFloat4FromVectorAndFloat("Vector (RGB)"=&VectorFromFloat(Value=25), "Float (Alpha)"=0.25)
    )
    Lifetime=&UniformRangedFloat(Minimum=0.2, Maximum=0.6)
    Sprite Rotation Angle=45
    Sprite Rotation Mode=::NewEnumerator0
    Sprite Size Mode=::NewEnumerator3
    Uniform Sprite Size=&UniformRangedFloat(Minimum=125, Maximum=75)
    UsePositionOffset=false
    Write Material Random=true
   ) {module=InitializeParticle001}
   &SystemLocation(MaskSpawnGroup=false, Offset=&FN_FrontEndVector3Modifier(Gameplay=<0 0 100>, Front End=<0 0 70>)) {enabled=false}
   &ShapeLocation(
    MaskSpawnGroup=false
    Offset Mode=::NewEnumerator0
    Override Seed=false
    Sphere Radius=15
    Sphere Surface Distribution=0.2
    UseOwnerScale=true
   )
   &ShapeLocation(
    Cylinder Height=500
    MaskSpawnGroup=false
    Offset=<50 0 0>
    Offset Coordinate Space=::Local
    Offset Mode=::NewEnumerator0
    Override Seed=false
    Rotation Mode=::NewEnumerator1
    Shape Primitive=::NewEnumerator1
    Surface Only=false
    UseOwnerScale=true
   ) {module=ShapeLocation001}
   &AddVelocity(
    Velocity=<1 0 0>
    Velocity Speed Scale=&Lerp_Float(
     A=500
     B=1200
     Alpha=&FloatFromCurve(
      FloatCurve=DataInterface() {propertyValues {Curve {keys [
          {interpMode=RCIM_Cubic, tangentMode=RCTM_User, time=0, value=0}
          {arriveTangent=1.7817, interpMode=RCIM_Cubic, leaveTangent=1.7817, tangentMode=RCTM_User, time=1, value=1}
         ]}}}
      CurveIndex=&RandomRangeFloat(Minimum=0, Maximum=1)
     )
    )
   )
   &AddVelocity(
    Module.Use Falloff Distance=false
    Override Seed=false
    Velocity Mode=::NewEnumerator1
    Velocity Origin=$Particles.ShapeLocation.ShapePosition
    Velocity Speed=&Lerp_Float(
     A=700
     B=1500
     Alpha=&FloatFromCurve(
      FloatCurve=DataInterface() {propertyValues {Curve {keys [
          {interpMode=RCIM_Cubic, tangentMode=RCTM_User, time=0, value=0}
          {arriveTangent=1.7817, interpMode=RCIM_Cubic, leaveTangent=1.7817, tangentMode=RCTM_User, time=1, value=1}
         ]}}}
      CurveIndex=&RandomRangeFloat(Minimum=0, Maximum=1)
     )
    )
   ) {module=AddVelocity001}
  ]
  particleUpdateScript [
   &ParticleState()
   &GravityForce(Gravity=<0 0 200>)
   &CurlNoiseForce(Noise Frequency=0.02, Noise Quality / Cost=::NewEnumerator0, Noise Strength=500, Pan Noise=false)
   &Drag(Drag=8, Use Linear Drag=true, Use Rotational Drag=true)
   &SolveForcesAndVelocity(Clamp Velocity=false, Limit Acceleration=false)
   &ScaleColor(
    Scale Alpha=&FloatFromCurve(FloatCurve=DataInterface() {propertyValues {Curve {keys [
         {arriveTangent=-2.9626, interpMode=RCIM_Cubic, leaveTangent=-2.9626, tangentMode=RCTM_User, time=0.3, value=1}
         {tangentMode=RCTM_User, time=1, value=0}
        ]}}})
    ScaleA=true
    ScaleRGB=true
   )
   &ScaleSpriteSize()
  ]
  rendererClasses [&NiagaraSpriteRendererProperties]
  renderers [{rendererClass=&NiagaraSpriteRendererProperties}]
 }
 Spooders_High {
  bEnabled=false
  simTarget=GPUComputeSim
  emitterUpdateScript [
   &EmitterState()
   &SpawnBurst_Instantaneous(
    Spawn Count=&Multiply_Int(
     A=35
     B=&MakeIntFromFloat(Float=&FN_FrontEndFloatModifier(Gameplay=1, Front End=0))
    )
    Use Spawn Probability=true
    UseLoopCountLimit=false
   )
  ]
  particleSpawnScript [
   &InitializeParticle(
    Color Mode=::NewEnumerator0
    Lifetime Max=1.5
    Lifetime Min=2.35
    Lifetime Mode=::NewEnumerator1
    Mesh Scale Mode=::NewEnumerator3
    Mesh Uniform Scale=0.1
    Sprite Size Mode=::NewEnumerator3
    Uniform Sprite Size=100
    UsePositionOffset=false
    Write Material Random=true
   )
   /SetVariables() {isSetParameters=true, module=SetVariables_#1}
   &ShapeLocation(MaskSpawnGroup=false, Override Seed=false, Sphere Radius=5, UseOwnerScale=true)
   &AddVelocity(
    Module.Use Falloff Distance=false
    Origin Offset=<0 0 -5>
    Override Seed=false
    Velocity Mode=::NewEnumerator1
    Velocity Speed=&RandomRangeFloat(Minimum=320, Maximum=500)
   )
   /SetVariables(Particles.prevPosition:NiagaraPosition=$Particles.Position) {isSetParameters=true, module=SetVariables_#2}
  ]
  particleUpdateScript [
   &ParticleState()
   &ScaleMeshSize(Scale Factor=&VectorFromFloat(Value=&FloatFromCurve(FloatCurve=DataInterface() {propertyValues {Curve {keys [
          {interpMode=RCIM_Cubic, tangentMode=RCTM_User, time=0, value=0}
          {interpMode=RCIM_Cubic, tangentMode=RCTM_User, time=0.1, value=1}
          {arriveTangentWeight=0.3986, interpMode=RCIM_Cubic, leaveTangentWeight=0.0498, tangentMode=RCTM_User, time=0.75, value=1}
          <at 1 0 RCIM_Cubic>
         ]}}})))
   &Collision(
    Age Colliding Particles=false
    Collision Enabled=&Boolean_NotOperation(Bool=$Particles.Is Crawling)
    GPU Collision Type=::NewEnumerator2
    KillOnCollision=false
    Module.Control Rest State Via Surface Normal=false
    Module.EnableMaxCollisionCount=false
    Module.Randomize Collision Normal Vector=false
    Particle Radius Scale=0.01
   )
   /SetVariables(Particles.Is Crawling:NiagaraBool=$Particles.HasCollided) {isSetParameters=true, module=SetVariables_#3}
   &InterpolateOverTime() {module=InterpolateOverTime001}
   &GravityForce(Gravity=&MakeCustomVectorFromBool(Bool=$Particles.Is Crawling, True Vector=<0 0 0>, False Vector=<0 0 -980>))
   /Niagara/Modules/Update/Forces/V2/CurlNoiseForce(
    Noise Frequency=120
    Noise Strength=&MakeCustomFloatFromBool(Bool=$Particles.Is Crawling, True Float=2500, False Float=0)
    Pan Noise=true
    Pan Noise Field=&RandomVector(Vector Scale=10)
   )
   /Niagara/Modules/Math/MoveToNearestDistanceFieldSurface_GPU(Active:NiagaraBool=$Particles.Is Crawling, bKillByDistance=false)
   &Drag(
    Drag=&MakeCustomFloatFromBool(Bool=$Particles.Is Crawling, True Float=0, False Float=0)
    Use Linear Drag=true
    Use Rotational Drag=true
   )
   &PointAttractionForce(
    Attraction Radius=999999995904
    Attraction Strength=&MakeCustomFloatFromBool(Bool=$Particles.Is Crawling, True Float=25, False Float=0)
    Kill Within Radius=false
    Module.Use Falloff=true
    UseAttractorOffset=false
   )
   &SolveForcesAndVelocity(Clamp Velocity=false, Limit Acceleration=false)
   &DynamicMaterialParameters(
    Index 0 Param 1=&VectorLength(Vector=$Particles.Velocity)
    Module.Index0Param0Write=true
    Module.Index0Param1Write=true
    Module.Index0Param2Write=true
    Module.Index0Param3Write=true
   )
   &InterpolateOverTime(
    Rate Of Change=5
    Target Vector Value:Vector3f=&self:ScratchDynamicInput()
    Variable Type=::NewEnumerator2
   )
   /SetVariables(Particles.DisplyVelocity:Vector3f=$Particles.InterpolateOverTime.InterpolateOverTime_V3_.Moving Average) {isSetParameters=true, module=SetVariables_#4}
   /SetVariables(Particles.MeshOrientation=&MultiplyQuaternion(
     Quaternion A=&MakeQuatFromAxisAngle(
      Axis=<0 0 1>
      AngleInDegrees=&Lerp_Float(A=-10, B=10, Alpha=&Cosine(Normalized Angle=$Emitter.Age, Period=0.1, Scale=0.5, Bias=0.5))
     )
     Quaternion B=$Particles.MeshOrientation
    )) {isSetParameters=true, module=SetVariables_#5}
  ]
  rendererClasses [&NiagaraMeshRendererProperties]
  renderers [{rendererClass=&NiagaraMeshRendererProperties, rendererIndex=0}]
 }
 Spooders_High_FE {
  simTarget=GPUComputeSim
  emitterUpdateScript [
   &EmitterState()
   &SpawnBurst_Instantaneous(
    Spawn Count=&Multiply_Int(
     A=20
     B=&MakeIntFromFloat(Float=&FN_FrontEndFloatModifier(Gameplay=0, Front End=1))
    )
    Use Spawn Probability=true
    UseLoopCountLimit=false
   )
  ]
  particleSpawnScript [
   &InitializeParticle(
    Color Mode=::NewEnumerator0
    Lifetime Max=1.5
    Lifetime Min=2.35
    Lifetime Mode=::NewEnumerator1
    Mesh Scale Mode=::NewEnumerator3
    Mesh Uniform Scale=0.065
    Position Offset=<0 0 6>
    Sprite Size Mode=::NewEnumerator3
    Uniform Sprite Size=100
    UsePositionOffset=true
    Write Material Random=true
   )
   /SetVariables() {isSetParameters=true, module=SetVariables_#1}
   &ShapeLocation(MaskSpawnGroup=false, Override Seed=false, Sphere Radius=5, UseOwnerScale=true)
   &AddVelocity(
    Module.Use Falloff Distance=false
    Origin Offset=<0 0 -5>
    Override Seed=false
    Velocity Mode=::NewEnumerator1
    Velocity Speed=&RandomRangeFloat(Minimum=320, Maximum=500)
   )
   /SetVariables(Particles.prevPosition:NiagaraPosition=$Particles.Position) {isSetParameters=true, module=SetVariables_#2}
  ]
  particleUpdateScript [
   &ParticleState()
   &ScaleMeshSize(Scale Factor=&VectorFromFloat(Value=&FloatFromCurve(FloatCurve=DataInterface() {propertyValues {Curve {keys [
          {interpMode=RCIM_Cubic, tangentMode=RCTM_User, time=0, value=0}
          {interpMode=RCIM_Cubic, tangentMode=RCTM_User, time=0.1, value=1}
          {arriveTangentWeight=0.3986, interpMode=RCIM_Cubic, leaveTangentWeight=0.0498, tangentMode=RCTM_User, time=0.75, value=1}
          <at 1 0 RCIM_Cubic>
         ]}}})))
   &GravityForce(Gravity=&MakeCustomVectorFromBool(Bool=$Particles.Is Crawling, True Vector=<0 0 0>, False Vector=<0 0 -980>))
   &Collision(
    Age Colliding Particles=false
    Analytical Collision Plane Position 1:Vector3f=<0 0 15>
    Analytical Collision Plane Space:ENiagaraCoordinateSpace=::World
    Collision Enabled=&Boolean_NotOperation(Bool=$Particles.Is Crawling)
    GPU Collision Type=::NewEnumerator3
    KillOnCollision=false
    Module.Control Rest State Via Surface Normal=false
    Module.EnableMaxCollisionCount=false
    Module.Randomize Collision Normal Vector=false
    Particle Radius Scale=0.01
   )
   /SetVariables(Particles.Is Crawling:NiagaraBool=$Particles.HasCollided) {isSetParameters=true, module=SetVariables_#3}
   &InterpolateOverTime() {module=InterpolateOverTime001}
   &CurlNoiseForce(
    Noise Frequency=120
    Noise Strength=&MakeCustomFloatFromBool(Bool=$Particles.Is Crawling, True Float=2500, False Float=0)
    Pan Noise=true
    Pan Noise Field=&RandomVector(Vector Scale=10)
   )
   &MoveToNearestDistanceFieldSurface_GPU(Active:NiagaraBool=$Particles.Is Crawling, bKillByDistance=false) {enabled=false}
   &Drag(
    Drag=&MakeCustomFloatFromBool(Bool=$Particles.Is Crawling, True Float=0, False Float=0)
    Use Linear Drag=true
    Use Rotational Drag=true
   )
   &PointAttractionForce(
    Attraction Radius=999999995904
    Attraction Strength=&MakeCustomFloatFromBool(Bool=$Particles.Is Crawling, True Float=25, False Float=0)
    Kill Within Radius=false
    Module.Use Falloff=true
    UseAttractorOffset=false
   )
   &SolveForcesAndVelocity(Clamp Velocity=false, Limit Acceleration=false)
   &DynamicMaterialParameters(
    Index 0 Param 1=&VectorLength(Vector=$Particles.Velocity)
    Module.Index0Param0Write=true
    Module.Index0Param1Write=true
    Module.Index0Param2Write=true
    Module.Index0Param3Write=true
   )
   &InterpolateOverTime(
    Rate Of Change=5
    Target Vector Value:Vector3f=&self:ScratchDynamicInput()
    Variable Type=::NewEnumerator2
   )
   /SetVariables(Particles.MeshOrientation=&MultiplyQuaternion(
     Quaternion A=&MakeQuatFromAxisAngle(
      Axis=<0 0 1>
      AngleInDegrees=&Lerp_Float(A=-10, B=10, Alpha=&Cosine(Normalized Angle=$Emitter.Age, Period=0.1, Scale=0.5, Bias=0.5))
     )
     Quaternion B=$Particles.MeshOrientation
    )) {isSetParameters=true, module=SetVariables_#5}
   /SetVariables(Particles.Position=&LerpPosition(
     A=$Particles.Position
     B=&self:posthing()
     Alpha=&MakeCustomFloatFromBool(Bool=$Particles.Is Crawling, True Float=1, False Float=0)
    )) {isSetParameters=true}
   /SetVariables(Particles.DisplyVelocity:Vector3f=&self:ScratchDynamicInput()) {isSetParameters=true, module=SetVariables_#4}
  ]
  rendererClasses [&NiagaraMeshRendererProperties]
  renderers [{rendererClass=&NiagaraMeshRendererProperties, rendererIndex=0}]
 }
 Spooders_Low {
  emitterUpdateScript [
   &EmitterState()
   &SpawnBurst_Instantaneous(Spawn Count=15, Use Spawn Probability=false, UseLoopCountLimit=false)
  ]
  particleSpawnScript [
   &InitializeParticle(
    Color Mode=::NewEnumerator0
    Lifetime Max=1
    Lifetime Min=1.5
    Lifetime Mode=::NewEnumerator1
    Mesh Scale Mode=::NewEnumerator3
    Mesh Uniform Scale=0.1
    Sprite Size Mode=::NewEnumerator3
    Uniform Sprite Size=100
    UsePositionOffset=false
    Write Material Random=true
   )
   /SetVariables() {isSetParameters=true, module=SetVariables_#1}
   &ShapeLocation(MaskSpawnGroup=false, Override Seed=false, Sphere Radius=5, UseOwnerScale=true)
   &AddVelocity(
    Module.Use Falloff Distance=false
    Origin Offset=<0 0 -5>
    Override Seed=false
    Velocity Mode=::NewEnumerator1
    Velocity Speed=&RandomRangeFloat(Minimum=600, Maximum=320)
   )
   /SetVariables(Particles.prevPosition:NiagaraPosition=$Particles.Position) {isSetParameters=true, module=SetVariables_#2}
  ]
  particleUpdateScript [
   &ParticleState()
   &ScaleMeshSize(Scale Factor=&VectorFromFloat(Value=&FloatFromCurve(FloatCurve=DataInterface() {propertyValues {Curve {keys [
          {interpMode=RCIM_Cubic, tangentMode=RCTM_User, time=0, value=0}
          {interpMode=RCIM_Cubic, tangentMode=RCTM_User, time=0.1, value=1}
          {arriveTangentWeight=0.3986, interpMode=RCIM_Cubic, leaveTangentWeight=0.0498, tangentMode=RCTM_User, time=0.75, value=1}
          <at 1 0 RCIM_Cubic>
         ]}}})))
   &GravityForce(Gravity=&MakeCustomVectorFromBool(Bool=$Particles.Is Crawling, True Vector=<0 0 0>, False Vector=<0 0 -980>))
   &SolveForcesAndVelocity(Clamp Velocity=false, Limit Acceleration=false)
   &DynamicMaterialParameters(
    Index 0 Param 1=&VectorLength(Vector=$Particles.Velocity)
    Module.Index0Param0Write=true
    Module.Index0Param1Write=true
    Module.Index0Param2Write=true
    Module.Index0Param3Write=true
   )
   &InterpolateOverTime(
    Rate Of Change=5
    Target Vector Value:Vector3f=&self:ScratchDynamicInput()
    Variable Type=::NewEnumerator2
   )
   /SetVariables(Particles.DisplyVelocity:Vector3f=$Particles.InterpolateOverTime.InterpolateOverTime_V3_.Moving Average) {isSetParameters=true, module=SetVariables_#4}
   /SetVariables(Particles.MeshOrientation=&MultiplyQuaternion(
     Quaternion A=&MakeQuatFromAxisAngle(
      Axis=<0 0 1>
      AngleInDegrees=&Lerp_Float(A=-10, B=10, Alpha=&Cosine(Normalized Angle=$Emitter.Age, Period=0.1, Scale=0.5, Bias=0.5))
     )
     Quaternion B=$Particles.MeshOrientation
    )) {isSetParameters=true, module=SetVariables_#5}
  ]
  rendererClasses [&NiagaraMeshRendererProperties]
  renderers [{rendererClass=&NiagaraMeshRendererProperties, rendererIndex=0}]
 }
}
summary {
 systemName=NS_Emote_Private_Jet_SpawnFE
 emitters [
  {emitterName=Smoke, rendererClasses [&NiagaraSpriteRendererProperties]}
  {emitterName=AirGlow, rendererClasses [&NiagaraSpriteRendererProperties]}
  {bEnabled=false, emitterName=Sparkles, rendererClasses [&NiagaraSpriteRendererProperties]}
  {bEnabled=false, emitterName=LargeSparkles001, rendererClasses [&NiagaraSpriteRendererProperties]}
  {bEnabled=false, emitterName=Spooders_High, simTarget=GPUComputeSim, rendererClasses [&NiagaraMeshRendererProperties]}
  {emitterName=Spooders_Low, rendererClasses [&NiagaraMeshRendererProperties]}
  {emitterName=Spooders_High_FE, simTarget=GPUComputeSim, rendererClasses [&NiagaraMeshRendererProperties]}
 ]
 userVariables [
  User.bIsFrontEnd {type=NiagaraBool, defaultValue=true}
  User.bIsFrontEndPreview {type=NiagaraBool, defaultValue=true}
 ]
}
